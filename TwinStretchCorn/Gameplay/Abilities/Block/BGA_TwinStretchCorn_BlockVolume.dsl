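; Player-held block shield for the TwinStretchCorn ability, a damageable volume with health that never dies.
; Forwards hits and depletion to the instigator as gameplay events, drives Niagara feedback, and heals by GameplayEffect while lowered.
; Replicated active state toggles collision and cues.
; Behind a hotfix flag, overlapping vehicles get server-computed, multicast impulses with per-actor cooldowns, while vehicle collision switches to ignore when the owner stands on one.
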(event Custom|CheckMovementBaseForVehicle
  (params (NewBaseComponent :type Primitive%20Component%20Object%20Reference) (BoneName :type Name))
  (bind _returnvalue (Components|GetOwner NewBaseComponent))
  (Utilities|IsValid NewBaseComponent
    (:Is%20Valid
      (Utilities|Casting|CastToFortVehicleInterface _returnvalue
        (:then
          (Variables|Default|SetbMovementBaseIsVehicle true)
          (Variables|Default|SetBaseVehicle _returnvalue)
          (Variables|Default|SetVehicleCollisionResponse "ECR_Ignore"))
        (:CastFailed
          _)))
    (:Is%20Not%20Valid
      _))
  (Variables|Default|SetbMovementBaseIsVehicle false)
  (Variables|Default|SetBaseVehicle 0)
  (Variables|Default|SetVehicleCollisionResponse "ECR_Block"))

(event EventBeginPlay
  (CallFunction|CheckVehicleRepelFeatureHotfix) ; This will turn off generate overlap events if the hotfix is off!
  (Health|BindEventtoOnBuildingHealthChanged (delegate Custom|OnBuildingHealthChanged_Event))
  (bind _asfort_player_pawn (Utilities|Casting|CastToFortPlayerPawn (Game|GetInstigator)))
  (bind _output_get (Variables|Default|SetOwningPawn _asfort_player_pawn))
  ;[1 1 1] Shield Hit Color Change
  (Ability|Async|WaitGameplayEventToActor (Game|GetInstigator) "(TagName=\"Event.TwinStretchCorn.Shield.Hit\")")
  (CallFunction|SetPlayerMovementIgnored true)
  (EventDispatchers|BindEventtoOnMovementBaseChanged (Variables|Default|GetOwningPawn) (EventDispatchers|CreateEvent :!event Custom|CheckMovementBaseForVehicle))
  (Ability|Tasks|WaitGameplayTagQueryOnActor _output_get "(TokenStreamVersion=0,TagDictionary=((TagName=\"Granted.TwinStretchCorn.Blocking\")),QueryTokenStream=(0,1,3,1,0),UserDescription=\"\",AutoDescription=\" NONE( Granted.TwinStretchCorn.Blocking )\")"
    (:then
      (bind (_asynctaskproxy _tagcount) (Ability|Tasks|WaitGameplayTagCountChangedOnActor (Variables|Default|GetOwningPawn) (Variables|Default|GetShieldCooldownTag))
        (:then
          (CallFunction|SetTeam (Team|GetActorTeam (Variables|Default|GetOwningPawn))))
        (:TagCountChanged
          (if (> _tagcount 0)
            (Variables|Default|SetNormalizedHealth 0.0)))))
    (:Triggered
      (CallFunction|StopGradualHealing))))

(event Custom|OnBuildingHealthChanged_Event
  (bind _returnvalue (Health|GetHealth))
  (Development|PrintString (Utilities|String|ToString(Float) _returnvalue) true true "(R=2.114420,G=0.000000,B=3.000000,A=1.000000)")
  (CallFunction|OnLocalHealthChanged _returnvalue)
  (Utilities|FlowControl|Sequence
    (:then_0
      (if (== (Health|GetHealthPercent) 1e+02)
        (CallFunction|StopGradualHealing)))
    (:then_1
      (if (<= _returnvalue 1.0)
        (if (not (Variables|Default|GetbShieldAlreadyBroke))
          (Variables|Default|SetbShieldAlreadyBroke true)
          (Default|CallOnOutOfHealth)
          (Ability|SendGameplayEventtoActor (Game|GetInstigator) (Variables|Default|GetShieldHealthEventTag)))
        (else
          (Variables|Default|SetbShieldAlreadyBroke false))))))

(event EventOnDamageServer
  (params (Damage :type Float%20%28single-precision%29) (DamageTags :type Gameplay%20Tag%20Container%20Structure%20%28by%20ref%29) (Momentum :type Vector :default "0, 0, 0") (HitInfo :type Hit%20Result%20Structure%20%28by%20ref%29) (InstigatedBy :type Controller%20Object%20Reference) (DamageCauser :type Actor%20Object%20Reference) (EffectContext :type Gameplay%20Effect%20Context%20Handle%20Structure))
  (bind _gameplayeventdata (Utilities|Struct|MakeGameplayEventData "(TagName=\"\")" InstigatedBy DamageCauser 0 0 "()" "(GameplayTags=)" "(GameplayTags=)" Damage))
  (Ability|SendGameplayEventtoActor (Game|GetInstigator) "(TagName=\"Event.TwinStretchCorn.Shield.Damage.Blocked\")" _gameplayeventdata)
  (Ability|SendGameplayEventtoActor (Game|GetInstigator) "(TagName=\"Event.TwinStretchCorn.Shield.Hit\")" _gameplayeventdata))

(event EventOnDamagePlayEffects
  (params (Damage :type Float%20%28single-precision%29) (DamageTags :type Gameplay%20Tag%20Container%20Structure%20%28by%20ref%29) (Momentum :type Vector :default "0, 0, 0") (HitInfo :type Hit%20Result%20Structure%20%28by%20ref%29) (InstigatedBy :type Fort%20Pawn%20Object%20Reference) (DamageCauser :type Actor%20Object%20Reference) (EffectContext :type Gameplay%20Effect%20Context%20Handle%20Structure))
  (bind _output_get (Variables|Default|SetCurrentHealth (Utilities|ClampMinFloat (- (Variables|Default|GetCurrentHealth) Damage))))
  (CallFunction|OnLocalHealthChanged _output_get)
  (CallFunction|ShieldHitVFX true))

(event Custom|ShieldHitVFX
  (params (IsDamage :type Boolean))
  (bind _shieldfx (Variables|Default|GetShieldFX))
  (Niagara|SetNiagaraVariable(Float) _shieldfx "HitTime" (Utilities|Time|GetGameTimeinSeconds))
  (Niagara|SetNiagaraVariable(Bool) _shieldfx "IsDamage" IsDamage)
  (Rendering|Material|SetCustomPrimitiveDataFloat (Variables|Default|GetShardsMesh) 1 (Utilities|Time|GetGameTimeinSeconds)))

(event Custom|OnLocalHealthChanged
  (params (CurrentHealth :type Float%20%28double-precision%29))
  (Variables|Default|SetNormalizedHealth (Health|GetHealthPercent)))

(event Custom|OnShieldPieceBroken
  (Utilities|IsValid (Utilities|Array|Get(acopy) (Variables|Default|GetShards) (Variables|Default|GetNumShards))
    (:Is%20Valid
      (Ability|ExecuteGameplayCueWithParamsReplicated (Variables|Default|GetCue_PieceBreak) (Ability|GameplayCue|MakeGameplayCueParameters :NormalizedMagnitude 0.0 :RawMagnitude 0.0 :Location (Math|Transform|TransformLocation (Transformation|GetActorTransform) (Utilities|Array|Get(acopy) (Variables|Default|GetShardLocations) (Variables|Default|GetNumShards))) :Normal "0,0,0" :GameplayEffectLevel 1 :AbilityLevel 1 :bReplicateLocationWhenUsingMinimalRepProxy false)))))

(event Custom|PlayIdleMontage)

(event Custom|MulticastCollisionUpdate
  (params (CollisionResponse :type ECollisionResponse%20Enum :default "ECR_Ignore"))
  ;[0.4583 0 0] Vehicle Deflecting Behavior
  (Collision|SetCollisionResponsetoChannel (Variables|Default|GetBox) "ECC_Vehicle" CollisionResponse))

(event Custom|PutVehicleImpulseOnCooldown
  (params (LaunchActor :type Actor%20Object%20Reference))
  ;[0.4583 0 0] Vehicle Deflecting Behavior
  ;;[1 1 1] Cooldown to prevent Impulse Spam
  (Utilities|Set|Add (Variables|Default|GetLaunchedActor) LaunchActor)
  (Utilities|FlowControl|Delay (ScalableFloat|GetValueatLevel (Variables|Default|GetHF_ShieldImpulseCooldown)))
  (Utilities|Set|Remove (Variables|Default|GetLaunchedActor) LaunchActor))

(event Collision|EventActorBeginOverlap
  (params (OtherActor :type Actor%20Object%20Reference))
  (bind _owningpawn (Variables|Default|GetOwningPawn))
  ;[0.4583 0 0] Vehicle Deflecting Behavior
  (if (ScalableFloat|GetValueasBool (Variables|Default|GetHF_EnableVehicleRepelling))
    (bind _asfort_vehicle_interface (Utilities|Casting|CastToFortVehicleInterface OtherActor))
    (if (not (Utilities|Array|ContainsItem (Variables|Default|GetOverlapping%20Vehicles) OtherActor))
      (bind _returnvalue (CallFunction|IsMovementBaseVehicle OtherActor))
      (if _returnvalue
        ;;[1 1 1] Extra precautions to ensure we do not end up with no collision
        (bind _output_get (Variables|Default|SetVehicleCollisionResponse "ECR_Ignore"))
        (CallFunction|MulticastCollisionUpdate _output_get)
        (else
          (bind _output_get_2 (Variables|Default|SetVehicleCollisionResponse "ECR_Block"))
          (CallFunction|MulticastCollisionUpdate _output_get_2)))
      ;;[1 1 1] Fix for Boss Vehicles in Cage
      (bind _returnvalue_2 (GameplayTags|HasMatchingGameplayTag(Message) OtherActor "(TagName=\"Gameplay.Melee.BounceOffTarget\")"))
      (if _returnvalue_2
        _
        (elif (CallFunction|IsImpulseTargetPawnBase OtherActor)
          _
          (else
            (Networking|SwitchHasAuthority
              (:Authority
                (bind (_launchactor _launchvelocity) (Utilities|CalculateLaunchVelocityAwayfromSource (Actor|GetOwner) OtherActor 0 (Variables|Default|GetKnockback%20Min%20Yaw%20Vehicle) 0 (Variables|Default|GetKnockback%20Vertical%20Angle%20Vehicle) 0 (CallFunction|CalculateVehicleLaunch OtherActor _asfort_vehicle_interface)))
                (if (not (Utilities|Set|ContainsItem (Variables|Default|GetLaunchedActor) _launchactor))
                  (CallFunction|PutVehicleImpulseonCooldown _launchactor)
                  (CallFunction|MulticastImpulseVehicle _asfort_vehicle_interface _launchvelocity)
                  (Utilities|Array|Add (Variables|Default|GetOverlapping%20Vehicles) OtherActor)
                  (if (not (GameplayTags|HasAllMatchingGameplayTags _owningpawn (Variables|Default|GetHasVehiclesToBlockTag)))
                    (Ability|GameplayEffect|AddLooseGameplayTags _owningpawn (Variables|Default|GetHasVehiclesToBlockTag) true))))) ; This inconsistently runs on client, \nso better just have it on the server for consistency
          ))))))

(event Collision|EventActorEndOverlap
  (params (OtherActor :type Actor%20Object%20Reference))
  (bind _overlapping_vehicles (Variables|Default|GetOverlapping%20Vehicles))
  ;[0.4583 0 0] Vehicle Deflecting Behavior
  (if (ScalableFloat|GetValueasBool (Variables|Default|GetHF_EnableVehicleRepelling))
    (Utilities|Casting|CastToFortVehicleInterface OtherActor)
    (Utilities|Array|RemoveItem _overlapping_vehicles OtherActor)
    (if (Utilities|Array|IsEmpty _overlapping_vehicles)
      (Ability|GameplayEffect|RemoveLooseGameplayTags (Variables|Default|GetOwningPawn) (Variables|Default|GetHasVehiclesToBlockTag) true))))

(event Custom|Multicast_ImpulseVehicle
  (params (Target :type Fort%20Vehicle%20Interface%20Interface) (LinearVelocity :type Vector%20%28by%20ref%29))
  ;[0.4583 0 0] Vehicle Deflecting Behavior
  (Vehicle|ImpulseVehicle Target LinearVelocity (Math|Vector|VectorZero) false false false "None")
  (CallFunction|ShieldHitVFX false)
  (GameplayCue|ExecuteGameplayCue (Variables|Default|GetOwningPawn) (Variables|Default|GetGC_VehicleImpact)))

(event EventEndPlay
  (params (EndPlayReason :type EEndPlayReason%20Enum :default "Destroyed"))
  (CallFunction|SetPlayerMovementIgnored false))

(fn CalculateVehicleLaunch
  (params (Actor :type Actor%20Object%20Reference) (FortVehicleInterace :type Fort%20Vehicle%20Interface%20Interface))
  (return (Math|Float|SelectFloat (Math|Vector|VectorLength (+ (* (Vehicle|GetVehicleVelocity FortVehicleInterace) -1.4) (* (Transformation|GetActorUpVector Actor) 3.5e+02))) (Math|Vector|VectorLength (* (Transformation|GetActorForwardVector Actor) -7.5e+02)) (>= (Math|Vector|VectorLength (Vehicle|GetVehicleVelocity FortVehicleInterace)) 2.5e+02))))

(fn CheckPlayerinVehicle
  (params (Vehicle :type Fort%20Vehicle%20Interface%20Interface))
  (bind _asfort_player_pawn (Utilities|Casting|CastToFortPlayerPawn (Game|GetInstigator))
    (:then
      (return (Utilities|Array|ContainsItem (Vehicle|GetAllPassengers Vehicle) _asfort_player_pawn)))
    (:CastFailed
      (return false))))

(fn CheckVehicleRepelFeatureHotfix
  (if (not (ScalableFloat|GetValueasBool (Variables|Default|GetHF_EnableVehicleRepelling)))
    (Components|DestroyComponent (Variables|Default|GetVehicleTrigger))))

(fn DoInstantHeal
  (GameplayEffects|ApplyGameplayEffectToSelf :self (CallFunction|GetShieldAbilitySystemComponent) :GameplayEffectClass (Variables|Default|GetInstantHealGE) :Level 0.0))

(fn GetShieldAbilitySystemComponent
  (return (Ability|GetAbilitySystemComponent self)))

(fn GetShieldIsActive
  (return (Variables|Default|GetbShieldIsActive)))

(fn IsImpulseTargetPawnBase
  (params (Actor :type Object%20Reference))
  (bind _vehiclecollisionresponse (Variables|Default|GetVehicleCollisionResponse))
  (bind _returnvalue (Utilities|Enum|Equal(Enum) _vehiclecollisionresponse "ECR_Ignore"))
  (bind _basevehicle (Variables|Default|GetBaseVehicle))
  (Utilities|IsValid _basevehicle
    (:Is%20Valid
      (if (and _returnvalue (== Actor _basevehicle))
        (return true)
        (else
          (return false))))
    (:Is%20Not%20Valid
      (if _returnvalue
        (return true)
        (else
          (return false))))))

(fn IsMovementBaseVehicle
  (params (Vehicle :type Object%20Reference))
  (bind _returnvalue (FortPawn|GetMyMovementBaseActor (Variables|Default|GetOwningPawn)))
  (return (or (== _returnvalue Vehicle) (Variables|Default|GetbMovementBaseIsVehicle))))

(fn OnRep_bShieldIsActive
  (bind _self self)
  (bind _bshieldisactive (Variables|Default|GetbShieldIsActive))
  (Collision|SetCollisionEnabled (Variables|Default|GetBox) (select _bshieldisactive "QueryAndPhysics" "NoCollision"))
  (Collision|SetCollisionEnabled (Variables|Default|GetVehicleTrigger) (select _bshieldisactive "QueryOnly" "NoCollision"))
  (Development|PrintString (Utilities|String|Append "Shield is Active: " (Utilities|String|ToString(Boolean) _bshieldisactive)) true true (Math|Color|SelectColor "(R=0.194090,G=1.000000,B=0.000000,A=1.000000)" "(R=1.000000,G=0.245088,B=0.000000,A=1.000000)" _bshieldisactive))
  (Effects|Components|ParticleSystem|SetBoolParameter (Variables|Default|GetShieldFX) "DefenseMode" (Variables|Default|GetbShieldIsActive))
  (if (Variables|Default|GetbShieldIsActive)
    (Components|Activation|Activate (Variables|Default|GetFortPlayerViewFollower) false)
    (GameplayCue|ExecuteGameplayCue _self (Variables|Default|GetGC_ShieldFormed))
    (GameplayCue|AddGameplayCueOnActor(Looping) _self (Variables|Default|GetGCNL_ShieldActive) (Ability|GameplayCue|MakeGameplayCueParameters 0.0))
    (else
      (GameplayCue|RemoveGameplayCueOnActor(Looping) self (Variables|Default|GetGCNL_ShieldActive) (Ability|GameplayCue|MakeGameplayCueParameters 0.0)))))

(fn OnRep_NumShards)

(fn OnRep_VehicleCollisionResponse
  (Collision|SetCollisionResponsetoChannel (Variables|Default|GetBox) "ECC_Vehicle" (Variables|Default|GetVehicleCollisionResponse)))

(fn SetPlayerMovementIgnored
  (params (bShouldIgnore :type Boolean))
  (Collision|IgnoreActorwhenMoving (Utilities|Casting|CastToPrimitiveComponent (Variables|/Script/Engine.Actor|GetRootComponent (Variables|Default|GetOwningPawn))) self bShouldIgnore))

(fn SetIdleFxVisible
  (params (Visible :type Boolean))
  (bind _shieldfx (Variables|Default|GetShieldFX))
  (if Visible
    (Components|Activation|Activate _shieldfx false)
    (else
      (Components|Activation|Deactivate _shieldfx))))

(fn SetShieldActive
  (params (ShouldBeActive :type Boolean))
  (bind _output_get (Variables|Default|SetbShieldIsActive ShouldBeActive))
  (Niagara|SetNiagaraVariable(Bool) (Variables|Default|GetShieldFX) "DefenseMode" _output_get)
  (if _output_get
    (CallFunction|StopGradualHealing)
    (elif (>= (Health|GetHealthPercent) 1.0)
      _
      (else
        (CallFunction|StartGradualHealing)))))

(fn ShouldDie
  (params (Damage :type Float%20%28single-precision%29) (EventInstigator :type Controller%20Object%20Reference) (DamageCauser :type Actor%20Object%20Reference) (bIsTest :type Boolean) (EffectContext :type Gameplay%20Effect%20Context%20Handle%20Structure%20%28by%20ref%29))
  (return false))

(fn StartGradualHealing
  (Networking|SwitchHasAuthority
    (:Authority
      (if (not (GameplayTags|HasMatchingGameplayTag (Variables|Default|GetOwningPawn) "(TagName=\"Granted.TwinStretchCorn.Blocking\")"))
        (bind _returnvalue (GameplayEffects|ApplyGameplayEffectToSelf :self (CallFunction|GetShieldAbilitySystemComponent) :GameplayEffectClass (Variables|Default|GetGradualHealGE) :Level 0.0))
        (Variables|Default|SetGradualHealHandle _returnvalue)))))

(fn StopGradualHealing
  (Networking|SwitchHasAuthority
    (:Authority
      (GameplayEffects|RemoveActiveGameplayEffect (CallFunction|GetShieldAbilitySystemComponent) (Variables|Default|GetGradualHealHandle) -1))))

(fn ConstructionScript)
